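# Physics asset for the Assemble_K_Male pickaxe mesh, defining collision bodies and joints for eight dangling wire bones.
# A capsule on the two-handed handle bone anchors eight dyn_wire capsules, six small identical ones and two larger, offset and rotated ones.
# Each wire is constrained directly to the handle rather than chained, with narrow swing and almost no twist, so each moves independently within a tight range.
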PhysicsAsset /BRCosmetics/Weapons/FORT_Melee/Pickaxe_Assemble_K_Male/Meshes/Assemble_K_Male_Axe_Physics
bodies {
 Melee_TwoHanded_Handle [{shapeType=Capsule, length=10, radius=5}]
 dyn_wire_1 [{shapeType=Capsule, length=3.6997, radius=0.9111, shapeName=dyn_wire_1_capsule}]
 dyn_wire_2 [{shapeType=Capsule, length=3.6997, radius=0.9111, shapeName=dyn_wire_1_capsule}]
 dyn_wire_3 [{shapeType=Capsule, length=3.6997, radius=0.9111, shapeName=dyn_wire_1_capsule}]
 dyn_wire_4 [{shapeType=Capsule, length=3.6997, radius=0.9111, shapeName=dyn_wire_1_capsule}]
 dyn_wire_5 [{shapeType=Capsule, length=3.6997, radius=0.9111, shapeName=dyn_wire_1_capsule}]
 dyn_wire_6 [{shapeType=Capsule, length=3.6997, radius=0.9111, shapeName=dyn_wire_1_capsule}]
 dyn_wire_7 [{
   shapeType=Capsule
   length=5.9923
   radius=2.0588
   shapeName=dyn_wire_7_capsule
   center=<0.1015 -0.1969 1.4821>
   rotation=<rot -23.2843 -90 6.5242>
  }]
 dyn_wire_8 [{
   shapeType=Capsule
   length=5.9923
   radius=2.0588
   shapeName=dyn_wire_7_capsule
   center=<0.1015 -0.1969 1.4821>
   rotation=<rot -23.2843 -90 6.5242>
  }]
}
constraints [
 {bone1Name=dyn_wire_1, bone2Name=Melee_TwoHanded_Handle, swing1LimitDegrees=10, swing2LimitDegrees=20, twistLimitDegrees=1}
 {bone1Name=dyn_wire_2, bone2Name=Melee_TwoHanded_Handle, swing1LimitDegrees=10, swing2LimitDegrees=20, twistLimitDegrees=1}
 {bone1Name=dyn_wire_3, bone2Name=Melee_TwoHanded_Handle, swing1LimitDegrees=10, swing2LimitDegrees=20, twistLimitDegrees=1}
 {bone1Name=dyn_wire_4, bone2Name=Melee_TwoHanded_Handle, swing1LimitDegrees=10, swing2LimitDegrees=20, twistLimitDegrees=1}
 {bone1Name=dyn_wire_5, bone2Name=Melee_TwoHanded_Handle, swing1LimitDegrees=10, swing2LimitDegrees=20, twistLimitDegrees=1}
 {bone1Name=dyn_wire_6, bone2Name=Melee_TwoHanded_Handle, swing1LimitDegrees=10, swing2LimitDegrees=20, twistLimitDegrees=1}
 {bone1Name=dyn_wire_7, bone2Name=Melee_TwoHanded_Handle, swing1LimitDegrees=10, swing2LimitDegrees=20, twistLimitDegrees=1}
 {bone1Name=dyn_wire_8, bone2Name=Melee_TwoHanded_Handle, swing1LimitDegrees=10, swing2LimitDegrees=20, twistLimitDegrees=1}
]
